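# Physics asset for the male medium CashmereScarf body cosmetic, setting up capsules and joints for its dangling jacket panels and chain.
# Six single-bone jacket panels have narrow swing and twist limits against small base anchors, which attach to spine_02 with free rotation.
# Front and back chains hang in limited segments from bases angularly locked to the right thigh, their tips joined into a loop.
# Larger capsules cover the lower spine, thighs and right calf.
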PhysicsAsset /BRCosmetics/Characters/Player/Male/Medium/Bodies/M_MED_CashmereScarf/Meshes/M_MED_CashmereScarf_Physics
bodies {
 base_chain_back [{shapeType=Capsule, length=0.505, radius=0.505, shapeName=base_chain_back_capsule, rotation=<rot 90 0 0>}]
 base_chain_front [{shapeType=Capsule, length=0.505, radius=0.505, shapeName=base_chain_front_capsule, rotation=<rot 90 0 0>}]
 base_jacket_front_l [{shapeType=Capsule, length=0.505, radius=0.505, shapeName=base_jacket_front_l_capsule, rotation=<rot 0 -90 0>}]
 base_jacket_front_r [{shapeType=Capsule, length=0.505, radius=0.505, shapeName=base_jacket_front_r_capsule, rotation=<rot 0 -90 0>}]
 base_jacket_l_back [{shapeType=Capsule, length=0.505, radius=0.505, shapeName=base_jacket_l_back_capsule, rotation=<rot 0 -90 0>}]
 base_jacket_l_mid [{shapeType=Capsule, length=0.505, radius=0.505, shapeName=base_jacket_l_mid_capsule, rotation=<rot 0 -90 0>}]
 base_jacket_r_back [{shapeType=Capsule, length=0.505, radius=0.505, shapeName=base_jacket_r_back_capsule, rotation=<rot 0 -90 0>}]
 base_jacket_r_mid [{shapeType=Capsule, length=0.505, radius=0.505, shapeName=base_jacket_r_mid_capsule, rotation=<rot 0 -90 0>}]
 calf_r [{shapeType=Capsule, length=33, radius=8, shapeName=thigh_r_capsule, center=<20 1.5 0>, rotation=<rot 90 0 0>}]
 dyn_chain_back_01 [{shapeType=Capsule, length=4, radius=2, shapeName=dyn_chain_back_01_capsule, center=<2 -2 0.5>, rotation=<rot 90 0 0>}]
 dyn_chain_back_02 [{shapeType=Capsule, length=4, radius=2.8, shapeName=dyn_chain_back_02_capsule, center=<2 -3 0.5>, rotation=<rot 90 0 0>}]
 dyn_chain_back_03 [{shapeType=Capsule, length=11, radius=2.8, shapeName=dyn_chain_back_03_capsule, center=<5 -3 1.2>, rotation=<rot 90 0 0>}]
 dyn_chain_front_01 [{shapeType=Capsule, length=3, radius=2, shapeName=dyn_chain_front_01_capsule, center=<3 -2.5 0>, rotation=<rot 90 0 0>}]
 dyn_chain_front_02 [{shapeType=Capsule, length=2.9818, radius=2, shapeName=dyn_chain_front_02_capsule, center=<2 -2 0>, rotation=<rot 90 0 0>}]
 dyn_chain_front_03 [{shapeType=Capsule, length=3.9184, radius=2.5, shapeName=dyn_chain_front_03_capsule, center=<2 -2 0>, rotation=<rot 90 0 0>}]
 dyn_chain_front_04 [{shapeType=Capsule, length=11, radius=2.5, shapeName=dyn_chain_front_04_capsule, center=<5 -2 -1.2>, rotation=<rot 90 0 0>}]
 dyn_jacket_l_back_01 [{shapeType=Capsule, length=11, radius=6, shapeName=dyn_jacket_l_back_01_capsule, center=<8 -4 0>, rotation=<rot 90 0 0>}]
 dyn_jacket_l_front_01 [{shapeType=Capsule, length=11, radius=6, shapeName=dyn_jacket_l_front_01_capsule, center=<13 -4 -2>}]
 dyn_jacket_l_mid_01 [{shapeType=Capsule, length=11, radius=6, shapeName=dyn_jacket_l_mid_01_capsule, center=<8 -5 0>, rotation=<rot 90 0 0>}]
 dyn_jacket_r_back_01 [{shapeType=Capsule, length=11, radius=6, shapeName=dyn_jacket_r_back_01_capsule, center=<8 -4 0>, rotation=<rot 90 0 0>}]
 dyn_jacket_r_front_01 [{shapeType=Capsule, length=24, radius=6.8, shapeName=dyn_jacket_r_front_01_capsule, center=<13 -4 2>, rotation=<rot 0 0 -20>}]
 dyn_jacket_r_mid_01 [{shapeType=Capsule, length=24, radius=6.8, shapeName=dyn_jacket_r_mid_01_capsule, center=<8 -4.5 0>, rotation=<rot 90 0 0>}]
 spine_01 [{shapeType=Capsule, length=9, radius=11, shapeName=spine_01_capsule}]
 spine_02 [{shapeType=Capsule, length=9, radius=11, shapeName=spine_02_capsule}]
 thigh_l [{shapeType=Capsule, length=33, radius=11, shapeName=thigh_r_capsule, center=<-20 0 1>, rotation=<rot 90 0 0>}]
 thigh_r [{shapeType=Capsule, length=33, radius=10.2, shapeName=thigh_r_capsule, center=<20 0 -1>, rotation=<rot 88 0 0>}]
}
constraints [
 {bone1Name=dyn_chain_back_01, bone2Name=base_chain_back, swing1LimitDegrees=20, swing2LimitDegrees=30, twistLimitDegrees=10}
 {bone1Name=dyn_chain_back_02, bone2Name=dyn_chain_back_01, swing1LimitDegrees=20, swing2LimitDegrees=30, twistLimitDegrees=10}
 {bone1Name=dyn_chain_back_03, bone2Name=dyn_chain_back_02, swing1LimitDegrees=20, swing2LimitDegrees=20, twistLimitDegrees=10}
 {bone1Name=dyn_chain_front_01, bone2Name=base_chain_front, swing1LimitDegrees=20, swing2LimitDegrees=30, twistLimitDegrees=10}
 {bone1Name=dyn_chain_front_02, bone2Name=dyn_chain_front_01, swing1LimitDegrees=20, swing2LimitDegrees=30, twistLimitDegrees=10}
 {bone1Name=dyn_chain_front_03, bone2Name=dyn_chain_front_02, swing1LimitDegrees=20, swing2LimitDegrees=20, twistLimitDegrees=10}
 {bone1Name=dyn_chain_front_04, bone2Name=dyn_chain_front_03, swing1LimitDegrees=20, swing2LimitDegrees=20, twistLimitDegrees=10}
 {
  bone1Name=base_chain_back
  bone2Name=thigh_r
  swing1LimitDegrees=45
  swing1Motion=Locked
  swing2LimitDegrees=45
  swing2Motion=Locked
  twistLimitDegrees=45
  twistMotion=Locked
 }
 {
  bone1Name=base_chain_front
  bone2Name=thigh_r
  swing1LimitDegrees=45
  swing1Motion=Locked
  swing2LimitDegrees=45
  swing2Motion=Locked
  twistLimitDegrees=45
  twistMotion=Locked
 }
 {
  bone1Name=spine_02
  bone2Name=spine_01
  swing1LimitDegrees=45
  swing1Motion=Locked
  swing2LimitDegrees=45
  swing2Motion=Locked
  twistLimitDegrees=45
  twistMotion=Locked
 }
 {
  bone1Name=base_jacket_front_r
  bone2Name=spine_02
  swing1LimitDegrees=45
  swing1Motion=Free
  swing2LimitDegrees=45
  swing2Motion=Free
  twistLimitDegrees=45
  twistMotion=Free
 }
 {bone1Name=dyn_jacket_r_front_01, bone2Name=base_jacket_front_r, swing1LimitDegrees=20, swing2LimitDegrees=20, twistLimitDegrees=3}
 {
  bone1Name=base_jacket_front_l
  bone2Name=spine_02
  swing1LimitDegrees=45
  swing1Motion=Free
  swing2LimitDegrees=45
  swing2Motion=Free
  twistLimitDegrees=45
  twistMotion=Free
 }
 {bone1Name=dyn_jacket_l_front_01, bone2Name=base_jacket_front_l, swing1LimitDegrees=30, swing2LimitDegrees=20, twistLimitDegrees=3}
 {
  bone1Name=base_jacket_l_back
  bone2Name=spine_02
  swing1LimitDegrees=45
  swing1Motion=Free
  swing2LimitDegrees=45
  swing2Motion=Free
  twistLimitDegrees=45
  twistMotion=Free
 }
 {bone1Name=dyn_jacket_l_back_01, bone2Name=base_jacket_l_back, swing1LimitDegrees=15, swing2LimitDegrees=5, twistLimitDegrees=3}
 {
  bone1Name=base_jacket_l_mid
  bone2Name=spine_02
  swing1LimitDegrees=45
  swing1Motion=Free
  swing2LimitDegrees=45
  swing2Motion=Free
  twistLimitDegrees=45
  twistMotion=Free
 }
 {bone1Name=dyn_jacket_l_mid_01, bone2Name=base_jacket_l_mid, swing1LimitDegrees=15, swing2LimitDegrees=10, twistLimitDegrees=3}
 {
  bone1Name=base_jacket_r_back
  bone2Name=spine_02
  swing1LimitDegrees=45
  swing1Motion=Free
  swing2LimitDegrees=45
  swing2Motion=Free
  twistLimitDegrees=45
  twistMotion=Free
 }
 {bone1Name=dyn_jacket_r_back_01, bone2Name=base_jacket_r_back, swing1LimitDegrees=15, swing2LimitDegrees=5, twistLimitDegrees=3}
 {
  bone1Name=base_jacket_r_mid
  bone2Name=spine_02
  swing1LimitDegrees=45
  swing1Motion=Free
  swing2LimitDegrees=45
  swing2Motion=Free
  twistLimitDegrees=45
  twistMotion=Free
 }
 {bone1Name=dyn_jacket_r_mid_01, bone2Name=base_jacket_r_mid, swing1LimitDegrees=15, swing2LimitDegrees=10, twistLimitDegrees=3}
 {
  bone1Name=calf_r
  bone2Name=thigh_r
  swing1LimitDegrees=45
  swing1Motion=Free
  swing2LimitDegrees=45
  swing2Motion=Free
  twistLimitDegrees=45
  twistMotion=Free
 }
 {
  bone1Name=dyn_chain_front_04
  bone2Name=dyn_chain_back_03
  swing1LimitDegrees=45
  swing1Motion=Free
  swing2LimitDegrees=45
  swing2Motion=Free
  twistLimitDegrees=45
  twistMotion=Free
 }
]
