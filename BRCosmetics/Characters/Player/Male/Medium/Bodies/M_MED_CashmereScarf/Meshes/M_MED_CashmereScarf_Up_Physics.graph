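# Physics asset for the upper mesh of the male medium cashmere scarf cosmetic, setting up capsule physics for its neck bones.
# Six dyn_neck bodies around the neck are each constrained to a tiny base_neck anchor capsule with tight swing and twist limits.
# Larger head, neck and upper-arm capsules approximate the body.
# Head and neck joints swing and twist freely, and only the right mid-back anchor is constrained to neck_01.
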PhysicsAsset /BRCosmetics/Characters/Player/Male/Medium/Bodies/M_MED_CashmereScarf/Meshes/M_MED_CashmereScarf_Up_Physics
bodies {
 base_neck_back [{shapeType=Capsule, length=0.505, radius=0.505, shapeName=base_neck_back_capsule, rotation=<rot 0 -90 0>}]
 base_neck_front_l [{shapeType=Capsule, length=0.505, radius=0.505, shapeName=base_neck_front_l_capsule, rotation=<rot 0 -90 0>}]
 base_neck_front_r [{shapeType=Capsule, length=0.505, radius=0.505, shapeName=base_neck_front_r_capsule, rotation=<rot 0 -90 0>}]
 base_neck_mid_l [{shapeType=Capsule, length=0.505, radius=0.505, shapeName=base_neck_mid_l_capsule, rotation=<rot 0 -90 0>}]
 base_neck_mid_r [{shapeType=Capsule, length=0.505, radius=0.505, shapeName=base_neck_mid_r_capsule, rotation=<rot 0 -90 0>}]
 base_neck_midback_r [{shapeType=Capsule, length=0.505, radius=0.505, shapeName=base_neck_midback_r_capsule, rotation=<rot 0 -90 0>}]
 dyn_neck_back [{shapeType=Capsule, length=6, radius=3, shapeName=dyn_neck_back_capsule, center=<2.3 0.3 0>}]
 dyn_neck_front_l [{shapeType=Capsule, length=11, radius=2, shapeName=dyn_neck_front_l_capsule, center=<3.5 -1.5 4>, rotation=<rot 33 0 -5>}]
 dyn_neck_front_r [{shapeType=Capsule, length=11, radius=2, shapeName=dyn_neck_front_l_capsule, center=<3.5 0 -4>, rotation=<rot -40 0 5>}]
 dyn_neck_mid_l [{shapeType=Capsule, length=11, radius=2, shapeName=dyn_neck_front_l_capsule, center=<3 0 0>}]
 dyn_neck_mid_r [{shapeType=Capsule, length=4, radius=2, shapeName=dyn_neck_front_l_capsule, center=<3 0 0>}]
 dyn_neck_midback_r [{shapeType=Capsule, length=4, radius=2, shapeName=dyn_neck_midback_r_capsule, center=<3 0 0>}]
 head [{shapeType=Capsule, length=8, radius=7, shapeName=head_capsule, center=<2.3419 2.6562 0>, rotation=<rot 90 180 -150>}]
 neck_01 [{shapeType=Capsule, length=1, radius=6, shapeName=neck_01_capsule, center=<0 1.57 0>, rotation=<rot 90 0 0>}]
 neck_02 [{shapeType=Capsule, length=1, radius=6, shapeName=neck_02_capsule, center=<0 1.57 0>, rotation=<rot 90 0 0>}]
 upperarm_l [{shapeType=Capsule, length=6, radius=9, shapeName=upperarm_l_capsule, center=<5 0 0>, rotation=<rot 90 0 0>}]
 upperarm_r [{shapeType=Capsule, length=6, radius=10, shapeName=upperarm_r_capsule, center=<-5 0 -2>, rotation=<rot 90 -88.7267 -88.7267>}]
}
constraints [
 {bone1Name=dyn_neck_front_l, bone2Name=base_neck_front_l, swing1LimitDegrees=20, swing2LimitDegrees=1, twistLimitDegrees=1}
 {bone1Name=dyn_neck_front_r, bone2Name=base_neck_front_r, swing1LimitDegrees=20, swing2LimitDegrees=1, twistLimitDegrees=1}
 {bone1Name=dyn_neck_mid_l, bone2Name=base_neck_mid_l, swing1LimitDegrees=15, swing2LimitDegrees=5, twistLimitDegrees=1}
 {bone1Name=dyn_neck_mid_r, bone2Name=base_neck_mid_r, swing1LimitDegrees=15, swing2LimitDegrees=5, twistLimitDegrees=1}
 {bone1Name=dyn_neck_back, bone2Name=base_neck_back, swing1LimitDegrees=15, swing2LimitDegrees=5, twistLimitDegrees=1}
 {
  bone1Name=neck_02
  bone2Name=neck_01
  swing1LimitDegrees=45
  swing1Motion=Free
  swing2LimitDegrees=45
  swing2Motion=Free
  twistLimitDegrees=45
  twistMotion=Free
 }
 {
  bone1Name=head
  bone2Name=neck_02
  swing1LimitDegrees=45
  swing1Motion=Free
  swing2LimitDegrees=45
  swing2Motion=Free
  twistLimitDegrees=45
  twistMotion=Free
 }
 {bone1Name=base_neck_midback_r, bone2Name=neck_01, swing1LimitDegrees=45, swing2LimitDegrees=45, twistLimitDegrees=45}
 {bone1Name=dyn_neck_midback_r, bone2Name=base_neck_midback_r, swing1LimitDegrees=15, swing2LimitDegrees=5, twistLimitDegrees=1}
]
