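# Animation Blueprint for the male cashmere scarf character body, adding secondary motion to the pose copied from the attached parent mesh.
# Bone constraints, a state-fed Control Rig, AnimDynamics bag-belt bones and additive offsets build a cached base pose.
# That feeds chain constraints and pelvis and neck RigidBody simulations, swapped for fixed jacket offsets when rigid bodies are disabled.
# The EventGraph updates that switch, gravity override, and crouch, skydive, pistol and emote state.
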AnimBlueprint /BRCosmetics/Characters/Player/Male/Medium/Bodies/M_MED_CashmereScarf/Meshes/M_MED_CashmereScarf_AnimBP
target_skeleton=&/M_MED_CashmereScarf_Skeleton
graphs [
 AnimGraph {nodes [
   n1 = "Constraint - base_chain_front (PPPPPPP)"!(
    ComponentPose=Local To Component(LocalPose=Use cached pose 'CACHE_Start'())
    Alpha=1
    ConstraintWeights_0=0.451
    ConstraintWeights_1:float=0.2719
    ConstraintWeights_2:float=0.2654
    ConstraintWeights_3:float=0.0066
    ConstraintWeights_4:float=0.0039
    ConstraintWeights_5:float=0.0009
    ConstraintWeights_6:float=0.0002
   ) -> (Pose) {
    alphaBoolBlend {blendOption=Linear}
    boneToModify {boneName=base_chain_front}
    constraintSetup [
     {targetBone {boneName=thigh_twist_01_r}}
     {targetBone {boneName=pelvis}}
     {targetBone {boneName=deform_groin_r}}
     {targetBone {boneName=thigh_r}}
     {targetBone {boneName=thigh_twist_01_l}}
     {targetBone {boneName=deform_groin_l}}
     {targetBone {boneName=deform_glute_r}}
    ]
    constraintWeights [1,1,1,1,1,1,1]
   }
   n2 = "Constraint - base_chain_back (PPPPPPPP)"!(
    ComponentPose=@n1
    Alpha=1
    ConstraintWeights_0=0.6458
    ConstraintWeights_1:float=0.1561
    ConstraintWeights_2:float=0.1171
    ConstraintWeights_3:float=0.065
    ConstraintWeights_4:float=0.0104
    ConstraintWeights_5:float=0.0045
    ConstraintWeights_6:float=0.0006
    ConstraintWeights_7:float=0.0004
   ) -> (Pose) {
    alphaBoolBlend {blendOption=Linear}
    boneToModify {boneName=base_chain_back}
    constraintSetup [
     {targetBone {boneName=thigh_twist_01_r}}
     {targetBone {boneName=deform_glute_r}}
     {targetBone {boneName=pelvis}}
     {targetBone {boneName=thigh_r}}
     {targetBone {boneName=thigh_twist_01_l}}
     {targetBone {boneName=deform_glute_l}}
     {targetBone {boneName=deform_groin_r}}
     {targetBone {boneName=thigh_l}}
    ]
    constraintWeights [1,1,1,1,1,1,1,1]
   }
   n3 = "Transform (Modify) Bone - Bone: base_jacket_front_r"!(ComponentPose=@n2, bAlphaBoolEnabled:bool=True, Translation="0,2,0", Rotation="0,0,0", Scale="1,1,1") -> (Pose) {
    alphaInputType=Bool
    rotationMode=BMM_Additive
    rotationSpace=BCS_BoneSpace
    scaleMode=BMM_Additive
    scaleSpace=BCS_BoneSpace
    translationMode=BMM_Additive
    translationSpace=BCS_BoneSpace
    alphaBoolBlend {blendInTime=0.5, blendOption=Sinusoidal, blendOutTime=1}
    boneToModify {boneName=base_jacket_front_r}
    translation {y=2}
   }
   n4 = "Transform (Modify) Bone - Bone: base_chain_front"!(ComponentPose=@n3, bAlphaBoolEnabled:bool=True, Translation="0,0,0", Rotation="30,0,0", Scale="1,1,1") -> (Pose) {
    alphaInputType=Bool
    rotationMode=BMM_Additive
    rotationSpace=BCS_BoneSpace
    scaleMode=BMM_Additive
    scaleSpace=BCS_BoneSpace
    translationMode=BMM_Additive
    translationSpace=BCS_BoneSpace
    alphaBoolBlend {blendInTime=0.5, blendOption=Sinusoidal, blendOutTime=1}
    boneToModify {boneName=base_chain_front}
    rotation {pitch=30}
   }
   n5 = "Transform (Modify) Bone - Bone: base_chain_front"(ComponentPose=@n4, bAlphaBoolEnabled=True, Translation="0,0,0", Rotation="0,0,45", Scale="1,1,1") {
    alphaInputType=Bool
    rotationMode=BMM_Additive
    rotationSpace=BCS_BoneSpace
    scaleMode=BMM_Additive
    scaleSpace=BCS_BoneSpace
    translationMode=BMM_Additive
    translationSpace=BCS_BoneSpace
    alphaBoolBlend {blendInTime=0.5, blendOption=Sinusoidal, blendOutTime=1}
    boneToModify {boneName=base_chain_front}
    rotation {roll=45}
   }
   n6 = Get GravityOverride()
   n7 = RigidBody!(ComponentPose=@n5, Alpha, OverrideWorldGravity:Vector=@n6, ExternalForce) -> (Pose) {
    bClampLinearTranslationLimitToRefPose=true
    bOverrideWorldGravity=true
    overridePhysicsAsset=&/M_MED_CashmereScarf_Physics
    simulationSpace=BaseBoneSpace
    alphaBoolBlend {blendOption=Linear}
    baseBoneRef {boneName=pelvis}
    componentAppliedLinearAccClamp=<3000 3000 3000>
    componentLinearAccScale=<0.6 0.6 0.6>
    componentLinearVelScale=<0.6 0.6 0.6>
    simSpaceSettings {
     maxAngularAcceleration=1000
     maxAngularVelocity=3
     maxLinearAcceleration=700
     maxLinearVelocity=500
     velocityScaleZ=0
     worldAlpha=0.2
     externalLinearDragV=<0.2 0.2>
    }
   }
   n8 = RigidBody(ComponentPose=@n7, OverrideWorldGravity=@n6) {
    bClampLinearTranslationLimitToRefPose=true
    bOverrideWorldGravity=true
    overridePhysicsAsset=&/M_MED_CashmereScarf_Up_Physics
    simulationSpace=BaseBoneSpace
    alphaBoolBlend {blendOption=Linear}
    baseBoneRef {boneName=neck_02}
    componentAppliedLinearAccClamp=<3000 3000 3000>
    componentLinearAccScale=<0.6 0.6 0.6>
    componentLinearVelScale=<0.6 0.6 0.6>
    simSpaceSettings {
     maxAngularAcceleration=1000
     maxAngularVelocity=3
     maxLinearAcceleration=700
     maxLinearVelocity=500
     velocityScaleZ=0
     worldAlpha=0.2
     externalLinearDragV=<0.2 0.2>
    }
   }
   n9 = "Transform (Modify) Bone - Bone: base_jacket_front_l"!(ComponentPose=@n8, bAlphaBoolEnabled:bool=True, Translation="0,0,0", Rotation="0,2,0", Scale="1,1,1") -> (Pose) {
    alphaInputType=Bool
    rotationMode=BMM_Additive
    rotationSpace=BCS_BoneSpace
    scaleMode=BMM_Additive
    scaleSpace=BCS_BoneSpace
    translationMode=BMM_Additive
    translationSpace=BCS_BoneSpace
    alphaBoolBlend {blendInTime=0.5, blendOption=Sinusoidal, blendOutTime=1}
    boneToModify {boneName=base_jacket_front_l}
    rotation {yaw=2}
   }
   Output Pose(Result=Blend Poses by bool(BlendPose_0=Use cached pose 'CACHE_ASWon'(), BlendPose_1=Component To Local(ComponentPose=@n9), bActiveValue=Get ASWon() -> (ASWon:bool)) {blendType=Linear}) {layerGroup=None}
   n10 = "Constraint - base_neck_front_r (PPP)"!(
    ComponentPose=Local To Component(LocalPose=Copy Pose From Mesh() {bUseAttachedParent=true})
    Alpha=1
    ConstraintWeights_0=0.5
    ConstraintWeights_1:float=0.25
    ConstraintWeights_2:float=0.25
   ) -> (Pose) {
    alphaBoolBlend {blendOption=Linear}
    boneToModify {boneName=base_neck_front_r}
    constraintSetup [{targetBone {boneName=clavicle_r}}, {targetBone {boneName=spine_05}}, {targetBone {boneName=neck_02}}]
    constraintWeights [1,1,1]
   }
   n11 = "Constraint - base_neck_mid_r (PPP)"!(ComponentPose=@n10, Alpha=1, ConstraintWeights_0=0.4, ConstraintWeights_1:float=0.3, ConstraintWeights_2:float=0.3) -> (Pose) {
    alphaBoolBlend {blendOption=Linear}
    boneToModify {boneName=base_neck_mid_r}
    constraintSetup [{targetBone {boneName=clavicle_r}}, {targetBone {boneName=spine_05}}, {targetBone {boneName=neck_02}}]
    constraintWeights [1,1,1]
   }
   n12 = "Constraint - base_neck_midback_r (PPP)"!(ComponentPose=@n11, Alpha=1, ConstraintWeights_0=0.3, ConstraintWeights_1:float=0.3, ConstraintWeights_2:float=0.4) -> (Pose) {
    alphaBoolBlend {blendOption=Linear}
    boneToModify {boneName=base_neck_midback_r}
    constraintSetup [{targetBone {boneName=clavicle_r}}, {targetBone {boneName=spine_05}}, {targetBone {boneName=neck_02}}]
    constraintWeights [1,1,1]
   }
   n13 = "Constraint - base_neck_back (PP)"!(ComponentPose=@n12, Alpha=1, ConstraintWeights_0=0.7, ConstraintWeights_1:float=0.3) -> (Pose) {
    alphaBoolBlend {blendOption=Linear}
    boneToModify {boneName=base_neck_back}
    constraintSetup [{targetBone {boneName=spine_05}}, {targetBone {boneName=neck_02}}]
    constraintWeights [1,1]
   }
   n14 = "Constraint - base_neck_mid_l (PPP)"!(ComponentPose=@n13, Alpha=1, ConstraintWeights_0=0.4, ConstraintWeights_1:float=0.3, ConstraintWeights_2:float=0.3) -> (Pose) {
    alphaBoolBlend {blendOption=Linear}
    boneToModify {boneName=base_neck_mid_l}
    constraintSetup [{targetBone {boneName=clavicle_l}}, {targetBone {boneName=spine_05}}, {targetBone {boneName=neck_02}}]
    constraintWeights [1,1,1]
   }
   n15 = "Constraint - base_neck_front_l (PPP)"!(ComponentPose=@n14, Alpha=1, ConstraintWeights_0=0.5, ConstraintWeights_1:float=0.25, ConstraintWeights_2:float=0.25) -> (Pose) {
    alphaBoolBlend {blendOption=Linear}
    boneToModify {boneName=base_neck_front_l}
    constraintSetup [{targetBone {boneName=clavicle_l}}, {targetBone {boneName=spine_05}}, {targetBone {boneName=neck_02}}]
    constraintWeights [1,1,1]
   }
   n16 = Control Rig!(
    Source=Component To Local(ComponentPose=@n15)
    Alpha
    IsCrouching:bool=Get bIsCrouching()
    Moving2D:bool=Get bIsMoving2D()
    Isnt Fly:double=Get CtrlRig_IsntFly() -> (CtrlRig_IsntFly:double)
    IsOnSky:bool=Get bIsSkyDiving()
    IsParachutOpen:bool=Get bIsParachuteOpened()
    IsTargeting:bool=Get bIsTargeting()
    IsPlayingMeleeAnimation:bool=Get bIsPlayingMeleeAnim()
    IsCustomEmotePlaying:bool=Get isCustomEmotePlaying() -> (isCustomEmotePlaying:bool)
   ) -> (Pose) {
    alphaBoolBlend {blendOption=Linear}
    controlRigAssetReference {blueprintRigClass=&/M_MED_CashmereScarf_CtrlRig.M_MED_CashmereScarf_CtrlRig_C}
   }
   n17 = "Constraint - base_bagbelt (PPPPPPP)"!(
    ComponentPose=Local To Component(LocalPose=@n16)
    Alpha=1
    ConstraintWeights_0=0.4405
    ConstraintWeights_1:float=0.2614
    ConstraintWeights_2:float=0.2019
    ConstraintWeights_3:float=0.0924
    ConstraintWeights_4:float=0.0021
    ConstraintWeights_5:float=0.001
    ConstraintWeights_6:float=0.0007
   ) -> (Pose) {
    alphaBoolBlend {blendOption=Linear}
    boneToModify {boneName=base_bagbelt}
    constraintSetup [
     {targetBone {boneName=spine_03}}
     {targetBone {boneName=spine_02}}
     {targetBone {boneName=pelvis}}
     {targetBone {boneName=spine_01}}
     {targetBone {boneName=clavicle_r}}
     {targetBone {boneName=deform_pec_l}}
     {targetBone {boneName=spine_05}}
    ]
    constraintWeights [1,1,1,1,1,1,1]
   }
   n18 = Anim Dynamics - Bone: dyn_bagbelt_01(ComponentPose=@n17, Alpha=1, ExternalForce="0,0,0") {
    angularDampingOverride=1
    angularSpringConstant=20000
    bAngularSpring=true
    bOverrideAngularDamping=true
    bUseGravityOverride=true
    gravityScale=0.35
    simulationSpace=RootRelative
    alphaBoolBlend {blendOption=Linear}
    boundBone {boneName=dyn_bagbelt_01}
    componentLinearAccScale=<1 1 1>
    componentLinearVelScale=<1 1 1>
    gravityOverride {z=-980}
    physicsBodyDefinitions [{
      boundBone {boneName=dyn_bagbelt_01}
      boxExtents=<20 30 30>
      constraintSetup {angularLimitsMax=<5 10 15>, angularLimitsMin=<-5 -10 -15>, angularTarget=<1 0 0>}
      localJointOffset=<4 0 0>
     }]
   }
   n19 = Anim Dynamics - Bone: dyn_bagbelt_02(ComponentPose=@n18, Alpha=1, ExternalForce="0,0,0") {
    angularDampingOverride=1
    angularSpringConstant=20000
    bAngularSpring=true
    bOverrideAngularDamping=true
    bUseGravityOverride=true
    gravityScale=0.35
    simulationSpace=RootRelative
    alphaBoolBlend {blendOption=Linear}
    boundBone {boneName=dyn_bagbelt_02}
    componentLinearAccScale=<1 1 1>
    componentLinearVelScale=<1 1 1>
    gravityOverride {z=-980}
    physicsBodyDefinitions [{
      boundBone {boneName=dyn_bagbelt_02}
      boxExtents=<20 30 30>
      constraintSetup {angularLimitsMax=<0 0 15>, angularLimitsMin=<0 0 -15>, angularTarget=<1 0 0>}
      localJointOffset=<4 0 0>
     }]
   }
   n20 = "Transform (Modify) Bone - Bone: base_jacket_front_l"(ComponentPose=@n19, bAlphaBoolEnabled=True, Translation="0,-6,0", Rotation="0,0,-20", Scale="1,1,1") {
    alphaInputType=Bool
    rotationMode=BMM_Additive
    rotationSpace=BCS_BoneSpace
    scaleMode=BMM_Additive
    scaleSpace=BCS_BoneSpace
    translationMode=BMM_Additive
    translationSpace=BCS_BoneSpace
    alphaBoolBlend {blendInTime=0.5, blendOption=Sinusoidal, blendOutTime=1}
    boneToModify {boneName=base_jacket_front_l}
    rotation {roll=-20}
    translation {y=-6}
   }
   n21 = "Transform (Modify) Bone - Bone: base_jacket_front_r"(ComponentPose=@n20, bAlphaBoolEnabled=True, Translation="-3,0,0", Rotation="0,0,0", Scale="1,1,1") {
    alphaInputType=Bool
    rotationMode=BMM_Additive
    rotationSpace=BCS_BoneSpace
    scaleMode=BMM_Additive
    scaleSpace=BCS_BoneSpace
    translationMode=BMM_Additive
    translationSpace=BCS_BoneSpace
    alphaBoolBlend {blendInTime=0.5, blendOption=Sinusoidal, blendOutTime=1}
    boneToModify {boneName=base_jacket_front_r}
    translation {x=-3}
   }
   n22 = "Transform (Modify) Bone - Bone: dyn_shoulderpad_back_r"!(ComponentPose=@n21, bAlphaBoolEnabled:bool=True, Translation="0.5,-0.5,0", Rotation="0,0,0", Scale="1,1,1") -> (Pose) {
    alphaInputType=Bool
    rotationMode=BMM_Additive
    rotationSpace=BCS_BoneSpace
    scaleMode=BMM_Additive
    scaleSpace=BCS_BoneSpace
    translationMode=BMM_Additive
    translationSpace=BCS_BoneSpace
    alphaBoolBlend {blendInTime=0.5, blendOption=Sinusoidal, blendOutTime=1}
    boneToModify {boneName=dyn_shoulderpad_back_r}
    translation=<0.5 -0.5>
   }
   n23 = "Transform (Modify) Bone - Bone: dyn_shoulderpad_back_r"(ComponentPose=@n22, bAlphaBoolEnabled=True, Translation="0,0,0", Rotation="0,5,0", Scale="1,1,1") {
    alphaInputType=Bool
    rotationMode=BMM_Additive
    rotationSpace=BCS_BoneSpace
    scaleMode=BMM_Additive
    scaleSpace=BCS_BoneSpace
    translationMode=BMM_Additive
    translationSpace=BCS_BoneSpace
    alphaBoolBlend {blendInTime=0.5, blendOption=Sinusoidal, blendOutTime=1}
    boneToModify {boneName=dyn_shoulderpad_back_r}
    rotation {yaw=5}
   }
   n24 = "Transform (Modify) Bone - Bone: attach_backpack"!(ComponentPose=@n23, bAlphaBoolEnabled:bool=True, Translation="0,0.5,0", Rotation="0,0,-6", Scale="1,1,1") -> (Pose) {
    alphaInputType=Bool
    rotationMode=BMM_Additive
    rotationSpace=BCS_BoneSpace
    scaleMode=BMM_Additive
    scaleSpace=BCS_BoneSpace
    translationMode=BMM_Additive
    translationSpace=BCS_BoneSpace
    alphaBoolBlend {blendInTime=0.5, blendOption=Sinusoidal, blendOutTime=1}
    boneToModify {boneName=attach_backpack}
    rotation {roll=-6}
    translation {y=0.5}
   }
   n25 = "Transform (Modify) Bone - Bone: attach_backpack"(ComponentPose=@n24, bAlphaBoolEnabled=True, Translation="0,-1.5,0", Rotation="0,0,10", Scale="1,1,1") {
    alphaInputType=Bool
    rotationMode=BMM_Additive
    rotationSpace=BCS_BoneSpace
    scaleMode=BMM_Additive
    scaleSpace=BCS_BoneSpace
    translationMode=BMM_Additive
    translationSpace=BCS_BoneSpace
    alphaBoolBlend {blendInTime=0.5, blendOption=Sinusoidal, blendOutTime=1}
    boneToModify {boneName=attach_backpack}
    rotation {roll=10}
    translation {y=-1.5}
   }
   Save cached pose 'CACHE_Start'(Pose=Component To Local(ComponentPose=@n25)) {cacheName=CACHE_Start}
   n26 = "Transform (Modify) Bone - Bone: base_jacket_front_l"(
    ComponentPose=Local To Component(LocalPose=Use cached pose 'CACHE_Start'())
    bAlphaBoolEnabled=True
    Translation="-2,0,0"
    Rotation="0,-48,0"
    Scale="1,1,1"
   ) {
    alphaInputType=Bool
    rotationMode=BMM_Additive
    rotationSpace=BCS_BoneSpace
    scaleMode=BMM_Additive
    scaleSpace=BCS_BoneSpace
    translationMode=BMM_Additive
    translationSpace=BCS_BoneSpace
    alphaBoolBlend {blendInTime=0.5, blendOption=Sinusoidal, blendOutTime=1}
    boneToModify {boneName=base_jacket_front_l}
    rotation {yaw=-48}
    translation {x=-2}
   }
   n27 = "Transform (Modify) Bone - Bone: base_jacket_front_l"(ComponentPose=@n26, bAlphaBoolEnabled=True, Translation="-5,0,0", Rotation="0,-18,0", Scale="1,1,1") {
    alphaInputType=Bool
    rotationMode=BMM_Additive
    rotationSpace=BCS_BoneSpace
    scaleMode=BMM_Additive
    scaleSpace=BCS_BoneSpace
    translationMode=BMM_Additive
    translationSpace=BCS_BoneSpace
    alphaBoolBlend {blendInTime=0.5, blendOption=Sinusoidal, blendOutTime=1}
    boneToModify {boneName=base_jacket_front_l}
    rotation {yaw=-18}
    translation {x=-5}
   }
   n28 = "Transform (Modify) Bone - Bone: base_jacket_front_r"(ComponentPose=@n27, bAlphaBoolEnabled=True, Translation="0,0,0", Rotation="0,-18,0", Scale="1,1,1") {
    alphaInputType=Bool
    rotationMode=BMM_Additive
    rotationSpace=BCS_BoneSpace
    scaleMode=BMM_Additive
    scaleSpace=BCS_BoneSpace
    translationMode=BMM_Additive
    translationSpace=BCS_BoneSpace
    alphaBoolBlend {blendInTime=0.5, blendOption=Sinusoidal, blendOutTime=1}
    boneToModify {boneName=base_jacket_front_r}
    rotation {yaw=-18}
   }
   n29 = "Transform (Modify) Bone - Bone: base_jacket_l_back"!(ComponentPose=@n28, bAlphaBoolEnabled:bool=True, Translation="0,0,0", Rotation="0,-25,0", Scale="1,1,1") -> (Pose) {
    alphaInputType=Bool
    rotationMode=BMM_Additive
    rotationSpace=BCS_BoneSpace
    scaleMode=BMM_Additive
    scaleSpace=BCS_BoneSpace
    translationMode=BMM_Additive
    translationSpace=BCS_BoneSpace
    alphaBoolBlend {blendInTime=0.5, blendOption=Sinusoidal, blendOutTime=1}
    boneToModify {boneName=base_jacket_l_back}
    rotation {yaw=-25}
   }
   n30 = "Transform (Modify) Bone - Bone: base_jacket_l_mid"!(ComponentPose=@n29, bAlphaBoolEnabled:bool=True, Translation="0,0,0", Rotation="0,-10,0", Scale="1,1,1") -> (Pose) {
    alphaInputType=Bool
    rotationMode=BMM_Additive
    rotationSpace=BCS_BoneSpace
    scaleMode=BMM_Additive
    scaleSpace=BCS_BoneSpace
    translationMode=BMM_Additive
    translationSpace=BCS_BoneSpace
    alphaBoolBlend {blendInTime=0.5, blendOption=Sinusoidal, blendOutTime=1}
    boneToModify {boneName=base_jacket_l_mid}
    rotation {yaw=-10}
   }
   n31 = "Transform (Modify) Bone - Bone: base_jacket_r_mid"!(ComponentPose=@n30, bAlphaBoolEnabled:bool=True, Translation="0,0,0", Rotation="0,-10,0", Scale="1,1,1") -> (Pose) {
    alphaInputType=Bool
    rotationMode=BMM_Additive
    rotationSpace=BCS_BoneSpace
    scaleMode=BMM_Additive
    scaleSpace=BCS_BoneSpace
    translationMode=BMM_Additive
    translationSpace=BCS_BoneSpace
    alphaBoolBlend {blendInTime=0.5, blendOption=Sinusoidal, blendOutTime=1}
    boneToModify {boneName=base_jacket_r_mid}
    rotation {yaw=-10}
   }
   n32 = "Transform (Modify) Bone - Bone: base_jacket_r_back"!(ComponentPose=@n31, bAlphaBoolEnabled:bool=True, Translation="0,0,0", Rotation="0,-25,0", Scale="1,1,1") -> (Pose) {
    alphaInputType=Bool
    rotationMode=BMM_Additive
    rotationSpace=BCS_BoneSpace
    scaleMode=BMM_Additive
    scaleSpace=BCS_BoneSpace
    translationMode=BMM_Additive
    translationSpace=BCS_BoneSpace
    alphaBoolBlend {blendInTime=0.5, blendOption=Sinusoidal, blendOutTime=1}
    boneToModify {boneName=base_jacket_r_back}
    rotation {yaw=-25}
   }
   Save cached pose 'CACHE_ASWon'(Pose=Component To Local(ComponentPose=@n32)) {cacheName=CACHE_ASWon}
  ]}
 EventGraph {nodes [
   Event BlueprintUpdateAnimation()
   > n1 = Sequence!(execute) -> (then_0, then_1, then_2:Exec, then_3:Exec, then_4:Exec, then_5:Exec, then_6:Exec, then_7:Exec)
   Set ASWon(execute=@n1.then_2, ASWon:bool=NOT Boolean(A=Get bRigidBodyEnabled())) -> (Output_Get:bool)
   n2 = Get bIsSkyDiving()
   Set CtrlRig_IsntFly(execute=@n1.then_5, CtrlRig_IsntFly:double="To Float (Boolean)"(InBool=NOT Boolean(A=@n2))) -> (Output_Get:double)
   n3 = Get bIsMoving2D()
   n4 = Get bIsCrouching()
   Set JustCrouching(execute=@n1.then_3, JustCrouching:bool=AND Boolean(A=NOT Boolean(A=@n3), B=@n4)) -> (Output_Get:bool)
   Set MoveCrouching(execute=@n1.then_4, MoveCrouching:bool=AND Boolean(A=@n3, B=@n4)) -> (Output_Get:bool)
   n5 = Gravity Override State Function DEPRECATED(
    execute=@n1.then_0
    CharacterPartAnimInstance=Self-Reference()
    GameTimeInSeconds=GetGameTimeInSeconds()
    GravityOverrideParams=Get Gravity Override Params()
   )
   > Set GravityOverride(GravityOverride=@n5)
   n6 = Cast To FortPlayerAnimInstance(execute=@n1.then_1, Object=GetAnimInstance(self=GetOwnerSkeletalMeshComponent()))
   > Set Pistol Equipped(Pistol Equipped:bool=false) -> (Output_Get:bool)
   > n7 = Switch on EFortWeaponCoreAnimation!(execute, Selection:EFortWeaponCoreAnimation Enum=Get WeaponCoreAnim(self=@n6)) -> (
    Melee:Exec
    Pistol:Exec
    Shotgun:Exec
    PaperBlueprint:Exec
    Rifle:Exec
    MeleeOneHand:Exec
    MachinePistol:Exec
    RocketLauncher:Exec
    GrenadeLauncher:Exec
    GoingCommando:Exec
    AssaultRifle:Exec
    TacticalShotgun:Exec
    SniperRifle:Exec
    TrapPlacement:Exec
    ShoulderLauncher:Exec
    AbilityDecoTool:Exec
    Crossbow:Exec
    C4:Exec
    RemoteControl:Exec
    DualWield:Exec
    AR_BullPup:Exec
    AR_ForwardGrip:Exec
    MedPackPaddles:Exec
    SMG_P90:Exec
    AR_DrumGun:Exec
    Consumable_Small:Exec
    Consumable_Large:Exec
    Balloon:Exec
    MountedTurret:Exec
    CreativeTool:Exec
    ExplosiveBow:Exec
    AshtonIndigo:Exec
    AshtonChicago:Exec
    MeleeDualWield:Exec
    Unarmed:Exec
   )
   n8 = Branch(execute=[@n7.DualWield, @n7.Pistol], Condition=Get bIsTargeting())
   > Set Pistol Equipped(Pistol Equipped=true)
   Set Pistol Equipped(execute=@n8.else, Pistol Equipped=false)
   Event OnBeginPlayingEmote() {is=title K2Node_Event}
   > Set isCustomEmotePlaying(isCustomEmotePlaying:bool=true) -> (Output_Get:bool)
   Event BlueprintInitializeAnimation()
   > Set Skel Mesh(Skel Mesh=GetOwnerSkeletalMeshComponent())
   Event OnEndPlayingEmote() {is=title K2Node_Event}
   > Set isCustomEmotePlaying(isCustomEmotePlaying=false)
  ]}
]
variables [
 ASWon {type=boolean}
 CtrlRig_IsntFly {type=number}
 Gravity Override Params {type=GravityOverrideParamsStruct}
 GravityOverride {type=Vector}
 isCustomEmotePlaying {type=boolean}
 JustCrouching {type=boolean}
 MoveCrouching {type=boolean}
 Pistol Equipped {type=boolean}
 Skel Mesh {type=SkeletalMeshComponent}
]
